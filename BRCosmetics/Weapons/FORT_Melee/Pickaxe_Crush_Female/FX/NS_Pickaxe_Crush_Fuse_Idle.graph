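# Niagara system for the idle fuse effect on the female Crush pickaxe cosmetic, with two continuously spawning emitters named for the left hand.
# Both re-sample a bone or socket location on User.SkeletalMesh every update, so particles stay attached to it.
# One renders longer-lived, randomly oriented mesh and sprite particles with warm orange color.
# The other renders very short-lived lights, using oscillating alpha and color-scale curves to create flicker.
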NiagaraSystem /BRCosmetics/Weapons/FORT_Melee/Pickaxe_Crush_Female/FX/NS_Pickaxe_Crush_Fuse_Idle
dependencies {
 usedDataInterfaces [&NiagaraDataInterfaceColorCurve, &NiagaraDataInterfaceCurve]
 usedDynamicInputs []
 usedModules [&SystemState]
}
emitters {
 FX_Fuse_L_Hand {
  emitterUpdateScript [
   &EmitterState(Life Cycle Mode=::NewEnumerator1, Loop Duration=1, UseLoopDelay=false)
   &SpawnBurst_Instantaneous(Spawn Count=3, Use Spawn Probability=false) {enabled=false}
   &SpawnRate(SpawnRate=6, Use Spawn Probability=false)
  ]
  particleSpawnScript [
   /Niagara/Modules/Spawn/Initialization/InitializeParticle(
    Color=<rgba 1 0.5316 0.0955 0.018>
    Lifetime=&RandomRangeFloat(Minimum=2, Maximum=2.5)
    Mesh Scale=&VectorFromFloat(Value=&Multiply_Float(A=0.031, B=&RandomRangeFloat(Minimum=1, Maximum=1.75)))
    Module.Write Color=true
    Module.Write Lifetime=true
    Module.Write Mass=true
    Module.Write Material Random=true
    Module.Write Position=false
    Module.Write Scale=true
    Module.Write SpriteRotation=true
    Module.Write SpriteSize=true
    Position=&SimulationPosition()
    Sprite Rotation=&RandomRangeFloat(Minimum=-1, Maximum=1)
    Sprite Size=&Vector2DFromFloat(Value=8)
   )
   &InitialMeshOrientation(
    Rotation=&RandomRangeVector(Minimum=<0 0 0>, Maximum=<1 1 1>)
    Use Orientation Vector=true
    Use Rotation Vector=true
   )
   &FN_SetSpawnGroup(Number of Spawn Groups=1)
  ]
  particleUpdateScript [
   &ParticleState()
   &SkeletalMeshLocation(
    Bone / Socket Index=&SpawnGroup()
    Inherit Velocity=false
    MaskSpawnGroup=false
    Mesh Sampling Type=::NewEnumerator5
    Sampled Position Offset Coordinate Space:ENiagaraMeshTransforms=::NewEnumerator2
    Skeletal Mesh=$User.SkeletalMesh
    Socket Sampling Mode=::NewEnumerator5
   )
   &SolveForcesAndVelocity(Clamp Velocity=false, Limit Acceleration=false)
   &FN_Light_Attributes(
    Color=&ColorFromCurve(ColorCurve=DataInterface() {propertyValues {
       AlphaCurve {keys [<at 0 0 RCIM_Cubic>,<at 0.279 1>,<at 1 0 RCIM_Cubic>]}
       BlueCurve {keys [<at 0 0 RCIM_Cubic>,<at 0.279 0>,<at 0.5266 0>,<at 1.0077 0.0776>]}
       GreenCurve {keys [<at 0 0.2118 RCIM_Cubic>,<at 0.279 0.3775>,<at 0.5266 0.5161>,<at 1.0077 0.4904>]}
       RedCurve {keys [<at 0 1 RCIM_Cubic>,<at 0.279 1>,<at 0.5266 1>,<at 1.0077 1>]}
      }})
    Create Color Attribute=true
    Create Exponent Attribute=true
    Create Radius Attribute=true
    Exponent=3
    Radius=12
    Scale Color=0.6
   )
  ]
  rendererClasses [&NiagaraMeshRendererProperties, &NiagaraSpriteRendererProperties]
  renderers [
   {rendererClass=&NiagaraMeshRendererProperties, rendererIndex=0}
   {rendererClass=&NiagaraSpriteRendererProperties, rendererIndex=1}
  ]
 }
 FX_Fuse_L_Hand001 {
  emitterUpdateScript [
   &EmitterState(Life Cycle Mode=::NewEnumerator1, Loop Duration=1, UseLoopDelay=false)
   &SpawnBurst_Instantaneous(Spawn Count=3, Use Spawn Probability=false) {enabled=false}
   &SpawnRate(SpawnRate=5, Use Spawn Probability=false)
  ]
  particleSpawnScript [
   &InitializeParticle(
    Color=<rgba 1 0.5316 0.0955 0.018>
    Lifetime=&RandomRangeFloat(Minimum=0.05, Maximum=0.1)
    Mesh Scale=&VectorFromFloat(Value=&Multiply_Float(A=0.031, B=&RandomRangeFloat(Minimum=1, Maximum=1.75)))
    Module.Write Color=true
    Module.Write Lifetime=true
    Module.Write Mass=true
    Module.Write Material Random=true
    Module.Write Position=false
    Module.Write Scale=true
    Module.Write SpriteRotation=true
    Module.Write SpriteSize=true
    Position=&SimulationPosition()
    Sprite Rotation=&RandomRangeFloat(Minimum=-1, Maximum=1)
    Sprite Size=&Vector2DFromFloat(Value=8)
   )
   &InitialMeshOrientation(
    Rotation=&RandomRangeVector(Minimum=<0 0 0>, Maximum=<1 1 1>)
    Use Orientation Vector=true
    Use Rotation Vector=true
   )
   &FN_SetSpawnGroup(Number of Spawn Groups=1)
  ]
  particleUpdateScript [
   &ParticleState()
   &SkeletalMeshLocation(
    Bone / Socket Index=&SpawnGroup()
    Inherit Velocity=false
    MaskSpawnGroup=false
    Mesh Sampling Type=::NewEnumerator5
    Sampled Position Offset Coordinate Space:ENiagaraMeshTransforms=::NewEnumerator2
    Skeletal Mesh=$User.SkeletalMesh
    Socket Sampling Mode=::NewEnumerator5
   )
   &SolveForcesAndVelocity(Clamp Velocity=false, Limit Acceleration=false)
   &FN_Light_Attributes(
    Color=&ColorFromCurve(ColorCurve=DataInterface() {propertyValues {
       AlphaCurve {keys [<at 0 0 RCIM_Cubic>,<at 0.1369 1>,<at 0.222 0.2>,<at 0.3111 1>,<at 0.4148 0.3>,<at 0.5144 1>,<at 0.6319 0.1>,<at 0.7481 1>,<at 1 0 RCIM_Cubic>]}
       BlueCurve {keys [<at 0 0.0727 RCIM_Cubic>,<at 0.279 0>,<at 0.5266 0>,<at 1.0077 0.0776>]}
       GreenCurve {keys [<at 0 0.3041 RCIM_Cubic>,<at 0.279 0.3775>,<at 0.5266 0.5161>,<at 1.0077 0.4904>]}
       RedCurve {keys [<at 0 1 RCIM_Cubic>,<at 0.279 1>,<at 0.5266 1>,<at 1.0077 1>]}
      }})
    Create Color Attribute=true
    Create Exponent Attribute=true
    Create Radius Attribute=true
    Exponent=3
    Radius=30
    Scale Color=&FloatFromCurve(FloatCurve=DataInterface() {propertyValues {Curve {keys [<at 0 0>,<at 0.0669 1.4>,<at 0.1386 0.1063>,<at 0.2125 1.4>,<at 0.2781 0.0876>,<at 0.3724 1.4>,<at 0.4872 0.0876>,<at 0.5815 1.4>,<at 0.682 0.0876>,<at 0.7722 1.4>,<at 1 0>]}}})
   )
  ]
  rendererClasses [&NiagaraLightRendererProperties]
  renderers [{rendererClass=&NiagaraLightRendererProperties, rendererIndex=0}]
 }
}
summary {
 systemName=NS_Pickaxe_Crush_Fuse_Idle
 emitters [
  {emitterName=FX_Fuse_L_Hand, rendererClasses [&NiagaraMeshRendererProperties, &NiagaraSpriteRendererProperties]}
  {emitterName=FX_Fuse_L_Hand001, rendererClasses [&NiagaraLightRendererProperties]}
 ]
 userVariables [User.SkeletalMesh {
   type=NiagaraDataInterfaceSkeletalMesh
   defaultValue {
    struct=NiagaraExt_VariableValue_DataInterface
    value {dataInterface=&self:NiagaraDataInterfaceSkeletalMesh_0}
   }
  }]
}
