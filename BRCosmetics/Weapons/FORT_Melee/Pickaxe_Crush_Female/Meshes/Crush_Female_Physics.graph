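# Physics asset for the Crush Female two-handed pickaxe, setting up its dangling heart, bow and fuse bones for physical motion.
# Capsule bodies cover the handle and each dynamic bone.
# Both hearts and the bow center constrain to the handle with swing1 locked, and the bow halves hang off the center with tight swing2 and twist limits.
# The fuse rotate bone constrains to the fuse translate bone, which has no constraint to the handle.
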PhysicsAsset /BRCosmetics/Weapons/FORT_Melee/Pickaxe_Crush_Female/Meshes/Crush_Female_Physics
bodies {
 Melee_TwoHanded_Handle [{shapeType=Capsule, length=45, radius=2.5, center=<0 0 18.0264>}]
 dyn_L_heart_jnt [{shapeType=Capsule, length=2, radius=2, center=<0 1.2519 0>}]
 dyn_R_heart_jnt [{shapeType=Capsule, length=2, radius=2, center=<0.0001 -1.428 0>}]
 dyn_bowCenter_jnt [{shapeType=Capsule, length=2.5, radius=3, center=<0 0 -0.1995>}]
 dyn_bow_L_jnt [{shapeType=Capsule, length=4.2269, radius=2.487, center=<2.1768 0.251 -0.1288>, rotation=<rot 0.2347 -90 -0.0646>}]
 dyn_bow_R_jnt [{shapeType=Capsule, length=4.5924, radius=2.297, center=<1.9801 -0.3461 0.4319>, rotation=<rot -0.1886 -90 3.9655>}]
 dyn_fus_rotate_jnt [{shapeType=Capsule, length=2, radius=1, center=<0.8965 -0.0109 -0.0019>, rotation=<rot 9.9995 89.2813 89.8746>}]
 dyn_fuse_translate_jnt [{shapeType=Capsule, length=4, radius=1, center=<-0.136 0 0>, rotation=<rot 0 0 -90>}]
}
constraints [
 {bone1Name=dyn_fus_rotate_jnt, bone2Name=dyn_fuse_translate_jnt, swing1LimitDegrees=30, swing2LimitDegrees=30, twistLimitDegrees=15}
 {
  bone1Name=dyn_R_heart_jnt
  bone2Name=Melee_TwoHanded_Handle
  swing1LimitDegrees=45
  swing1Motion=Locked
  swing2LimitDegrees=25
  twistLimitDegrees=15
 }
 {
  bone1Name=dyn_L_heart_jnt
  bone2Name=Melee_TwoHanded_Handle
  swing1LimitDegrees=45
  swing1Motion=Locked
  swing2LimitDegrees=25
  twistLimitDegrees=15
 }
 {bone1Name=dyn_bow_R_jnt, bone2Name=dyn_bowCenter_jnt, swing1LimitDegrees=45, swing2LimitDegrees=5, twistLimitDegrees=7}
 {bone1Name=dyn_bow_L_jnt, bone2Name=dyn_bowCenter_jnt, swing1LimitDegrees=45, swing2LimitDegrees=5, twistLimitDegrees=7}
 {
  bone1Name=dyn_bowCenter_jnt
  bone2Name=Melee_TwoHanded_Handle
  swing1LimitDegrees=45
  swing1Motion=Locked
  swing2LimitDegrees=2
  twistLimitDegrees=5
 }
]
