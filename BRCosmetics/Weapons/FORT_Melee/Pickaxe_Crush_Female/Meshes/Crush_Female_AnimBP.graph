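# Animation Blueprint that adds physics-driven secondary motion to the Crush Female pickaxe cosmetic.
# The input pose runs through a RigidBody node using the pickaxe physics asset, simulated in base-bone space around the two-handed handle bone.
# bIsInFrontEnd switches that node's gravity override and velocity scale between front-end and other contexts.
# An AnimDynamics node then lets the fuse joint slide slightly along Z in world-space simulation; the event graph has no logic.
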AnimBlueprint /BRCosmetics/Weapons/FORT_Melee/Pickaxe_Crush_Female/Meshes/Crush_Female_AnimBP
target_skeleton=&/Crush_Female_Skeleton
graphs [
 AnimGraph {nodes [
   AnimGraphNode_SubInput_0 = Input Pose()
   n1 = Get bIsInFrontEnd()
   n2 = RigidBody!(
    ComponentPose=Local To Component(LocalPose=@AnimGraphNode_SubInput_0)
    Alpha
    OverrideWorldGravity:Vector=SelectVector(A="-50,200,-1150", B="0,0,-700", bPickA=@n1)
    ExternalForce
    ComponentLinearAccScale:Vector="0.25,0.25,0.25"
    ComponentLinearVelScale:Vector=SelectVector(A="1,1,1", B="0.2,0.2,0.2", bPickA=@n1)
   ) -> (Pose) {
    bClampLinearTranslationLimitToRefPose=true
    bOverrideWorldGravity=true
    overridePhysicsAsset=&/Crush_Female_Physics
    simulationSpace=BaseBoneSpace
    alphaBoolBlend {blendOption=Linear}
    baseBoneRef {boneName=Melee_TwoHanded_Handle}
    componentAppliedLinearAccClamp=<4000 4000 4000>
    componentLinearAccScale=<0.25 0.25 0.25>
    componentLinearVelScale=<0.2 0.2 0.2>
    overrideWorldGravity {z=-700}
   }
   n3 = Anim Dynamics - Bone: dyn_fuse_translate_jnt(ComponentPose=@n2, Alpha=1, ExternalForce="0,0,0") {
    angularDampingOverride=0.999
    bGravityOverrideInSimSpace=true
    linearDampingOverride=0.999
    simulationSpace=World
    alphaBoolBlend {blendOption=Linear}
    boundBone {boneName=dyn_fuse_translate_jnt}
    componentAppliedLinearAccClamp=<6000 6000 6000>
    physicsBodyDefinitions [{
      sphereCollisionRadius=0
      boundBone {boneName=dyn_fuse_translate_jnt}
      boxExtents=<1 1 1>
      constraintSetup {angularLimitsMax=<0 0 0>, angularLimitsMin=<0 0 0>, angularTarget=<0 0 0>, linearAxesMax=<0 0 0.7>, linearAxesMin=<0 0 -0.7>}
      localJointOffset=<0 0 0>
     }]
    relativeSpaceBone {boneName=Melee_TwoHanded_Handle}
   }
   Output Pose(Result=Component To Local(ComponentPose=@n3)) {layerGroup=None}
  ]}
 EventGraph {nodes [
   TryGetPawnOwner()
   Event BlueprintUpdateAnimation()
  ]}
]
variables []
